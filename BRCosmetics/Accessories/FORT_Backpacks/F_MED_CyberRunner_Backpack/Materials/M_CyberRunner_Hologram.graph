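# Additive hologram material for the CyberRunner backpack cosmetic, drawing a glowing animated hex grid on skeletal meshes.
# The tiled hex mask blends two colors by Fresnel, while nested SphereMasks driven by looping time create a radial UV pulse that modulates color and brightness.
# Rotated scanlines multiply the result, the untiled mask's green channel adds edge color, and inverse Fresnel fades silhouettes before the shared accessory tech-art override function.
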Material /BRCosmetics/Accessories/FORT_Backpacks/F_MED_CyberRunner_Backpack/Materials/M_CyberRunner_Hologram
graph {
 properties {
  bUseMaterialAttributes=true
  bUsedWithMeshDeformer=true
  bUsedWithSkeletalMesh=true
  blendMode=BLEND_Additive
  ditherOpacityMask=true
  pixelDepthOffsetMode=PDOM_Legacy
  usedShadingModels=Default Lit
 }
 nodes [
  n1 = Fresnel(ExponentIn=ScalarParameter() {defaultValue=1.5, parameterName=HexFresnelEX})
  n2 = Saturate(None=Power(Base=@n1, Exp=ScalarParameter() {defaultValue=1.5, parameterName=EmissiveBrightness_Power}))
  n3 = VectorParameter() {parameterName=Hologram_Color_01, defaultValue=<rgba 0 0.1757 1 1>}
  n4 = VectorParameter() {parameterName=Hologram_Color_02, defaultValue=<rgba 0 6.7861 15 1>}
  n5 = Frac(None=Multiply(A=Time(), B=ScalarParameter() {defaultValue=0.3, parameterName=Pulse_Speed}))
  n6 = TextureCoordinate()
  n7 = ComponentMask(None=VectorParameter() {parameterName=Sphere Alignment, defaultValue=<rgba 0.11 0.18 0 1>}.RGB) {g=true, r=true}
  n8 = ScalarParameter() {defaultValue=0.15, parameterName=Radius_Shere}
  n9 = ScalarParameter() {parameterName=SphereHardness}
  n10 = Multiply(
   A=SphereMask(
    A=LinearInterpolate(Alpha=@n5) {constA=-0.5, constB=1.15}
    B=SphereMask(A=@n6, B=@n7, Radius=@n8, Hardness=@n9)
    Radius=ScalarParameter() {defaultValue=0.4, parameterName=Pulse_Radius}
    Hardness=Constant()
   )
   B=ScalarParameter() {defaultValue=0.6, parameterName=Pulse Brightness}
  )
  n11 = TextureCoordinate()
  n12 = TextureObject() {samplerType=SAMPLERTYPE_Masks, texture=/Game/Accessories/FORT_Backpacks/F_MED_CyberRunner_Backpack/Textures/HexGrid_M}
  n13 = TextureSample(UVs=Multiply(A=@n11, B=ScalarParameter() {defaultValue=8, parameterName=TilingHex}), Tex=@n12) {samplerType=SAMPLERTYPE_LinearGrayscale, texture=&HexGrid_M}
  n14 = LinearInterpolate(A=@n3.RGB, B=Add(A=LinearInterpolate(A=@n3.RGB, B=@n4.RGB, Alpha=@n1), B=@n10), Alpha=@n13.R)
  n15 = ScalarParameter() {defaultValue=2.5, parameterName=Hologram Brightness}
  n16 = ScalarParameter() {defaultValue=0.35, parameterName=Min_HoloGram_Brightness}
  n17 = LinearInterpolate(A=Constant() {r=0.1}, Alpha=Saturate(None=@n10))
  n18 = Multiply(A=@n13.R, B=@n17)
  n19 = TextureCoordinate()
  n20 = ScalarParameter() {defaultValue=0.25, parameterName=RotScanlines}
  n21 = Multiply(
   A=Multiply(A=Multiply(A=@n14, B=@n15), B=LinearInterpolate(A=@n16, B=@n17, Alpha=@n18))
   B=MaterialFunctionCall!(
    ScanLine_Tiling_1
    Scanline_PannerSpeed_1
    ScanLine_Tiling
    Scanline_PannerSpeed
    UVs=MaterialFunctionCall(UVs=@n19, "Rotation Angle (0-1)"=@n20) {materialFunction=&CustomRotator}.Rotated Values
   ) {materialFunction=/Game/Weapons/WeaponSkins/MaterialFunction/ScanLines}.Result
  )
  n22 = VectorParameter() {parameterName=EdgeColor, defaultValue=<rgba 0 4.1757 10 1>}
  n23 = TextureSample(UVs=@n11, Tex=@n12) {samplerType=SAMPLERTYPE_LinearGrayscale, texture=&HexGrid_M}
  n24 = SetMaterialAttributes!(
   MaterialAttributes
   Emissive Color=Multiply(A=LinearInterpolate(Alpha=@n2) {constA=1, constB=0}, B=LinearInterpolate(A=@n21, B=@n22.RGB, Alpha=@n23.G))
   Metallic
   Specular=Constant() {r=0.7}
   Roughness=Constant() {r=0.2}
   Opacity Mask
   Base Color=Constant()
  ) {attributeSetTypes [
    #1
    #2
    #3
    #4
    #5
    #6
   ]}
  MaterialFunctionCall(In=@n24) {materialFunction=&MF_TechArtOverrides_accessories}
 ]
}
